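# Niagara system for the TouchToken pickaxe's swing trail, built from short-lived animation-trail ribbons that follow bones.
# NE_Swing_01 feeds three ribbon renderers with wisp and main colors, while NE_Swing_01001 adds a single bluish fill ribbon using a different width mode.
# System fade values drive spawn and update alpha in both emitters.
# A curve looping on emitter age animates a dynamic material parameter.
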NiagaraSystem /BRCosmetics/Weapons/FORT_Melee/TouchToken/FX/NS_Pickaxe_TouchToken_Trail
dependencies {
 usedDataInterfaces [&NiagaraDataInterfaceColorCurve, &NiagaraDataInterfaceCurve, &NiagaraDataInterfaceSocketReader]
 usedDynamicInputs [
  &MakeVector4FromVectorAndFloat
  &MergeHelper_FloatSwitch
  &ClampFloat
  &RemapRange
  &Add_Float
  &Subtract_Float
 ]
 usedModules [
  &self:SystemSpawnScript.SystemScriptSource.SystemScriptGraph.NiagaraNodeAssignment_1.SetVariables_#1
  &self:SystemSpawnScript.SystemScriptSource.SystemScriptGraph.NiagaraNodeAssignment_0.SetVariables_#2
  &SystemState
  &self:SystemSpawnScript.SystemScriptSource.SystemScriptGraph.NiagaraNodeAssignment_2.SetVariables_#3
 ]
}
emitters {
 NE_Swing_01 {
  emitterSpawnScript [&FN_RibbonIDOffsetUtility()]
  emitterUpdateScript [
   &EmitterState()
   &FN_SpawnAnimTrail()
   /SetVariables(Emitter.Curve for Colors=DataInterface() {propertyValues {
      bExposeCurve=true
      AlphaCurve {keys [<at 0 0>,<at 0.25 0.5>,<at 0.5 0.25>,<at 0.75 0.5>,<at 1 0>]}
      BlueCurve {keys [<at 0 1>,<at 0.2216 0.8348>,<at 0.5122 0.787>,<at 0.575 1>,<at 1 0>]}
      GreenCurve {keys [<at 0 0.8913>,<at 0.2216 0.7513>,<at 0.5122 0.5612>,<at 0.575 0.8391>,<at 1 0>]}
      RedCurve {keys [<at 0 0.7826>,<at 0.2216 0.6678>,<at 0.5122 0.3353>,<at 0.575 0.6783>,<at 1 0>]}
     }}) {isSetParameters=true}
   /SetVariables(Emitter.Curve for Colors2=DataInterface() {propertyValues {
      bExposeCurve=true
      AlphaCurve {keys [<at 0 0>,<at 0.25 0.25>,<at 0.5 0.125>,<at 0.75 0.25>,<at 1 0>]}
      BlueCurve {keys [<at 0 1>,<at 1 0>]}
      GreenCurve {keys [<at 0 0.8026>,<at 1 0>]}
      RedCurve {keys [<at 0 0.5565>,<at 1 0>]}
     }}) {isSetParameters=true}
   /SetVariables(Emitter.Curve for Colors3=DataInterface() {propertyValues {
      bExposeCurve=true
      AlphaCurve {keys [<at 0 1>]}
      BlueCurve {keys [<at 0 0.8714>,<at 0.1744 0.2051>,<at 0.4618 0.2623>,<at 1 0.2051>]}
      GreenCurve {keys [<at 0 0.807>,<at 0.1744 0.0437>,<at 0.4618 0.0116>,<at 1 0.0086>]}
      RedCurve {keys [<at 0 0.7605>,<at 0.1744 0.013>,<at 0.4618 0.0012>,<at 1 0.0021>]}
     }}) {isSetParameters=true}
  ]
  particleSpawnScript [
   &FN_InitializeAnimTrailBoneIndecies()
   &FN_InitializeAnimTrail(
    AnimTrailWidth=1.15
    Color=&MakeLinearColorFromVectorAndFloat(
     "Vector (RGB)"=&MakeVectorFromLinearColorRGB(Color=<rgba 1 1 1 1>)
     "Float (Alpha)"=&Multiply_Float(A=$System.FadeSpawn, B=1)
    )
    Lifetime=0.13
   )
  ]
  particleUpdateScript [
   &FN_UpdateAnimTrail() {module=FN_UpdateAnimTrail001}
   &DynamicMaterialParameters(
    Index 0 Param 1=1
    Index 1 Param 3=&FloatFromCurve(
     FloatCurve=DataInterface() {propertyValues {Curve {keys [<at 0 0>, {arriveTangent=1.9814, tangentMode=RCTM_User, time=1, value=1}]}}}
     CurveIndex=&ModuloFloat(A=$Emitter.Age, B=1)
     Scale Curve=3
    )
    Index 1 Param 4=1
    Index 2 Param 1=10
    Index 2 Param 2=1
    Module.Index0Float0Write=true
    Module.Index0Float1Write=true
    Module.Index0Float2Write=true
    Module.Index0Float3Write=true
    Module.Index1Float0Write=true
    Module.Index1Float1Write=true
    Module.Index1Float2Write=true
    Module.Index1Float3Write=true
    Module.Index2Float0Write=true
    Module.Index2Float1Write=true
    Module.Index2Float2Write=true
    Module.Index2Float3Write=true
    Write Parameter Index 1=::NewEnumerator1
    Write Parameter Index 2=::NewEnumerator1
   )
   &ScaleColor(Scale Alpha=$System.FadeSystem, ScaleA=true, ScaleRGB=true) {module=ScaleColor002}
   /SetVariables(Particles.WispCol:LinearColor=&MultiplyLinearColor(
     Linear Color=&ColorFromCurve(ColorCurve=DataInterface() {propertyValues {
        AlphaCurve {keys [<at 0.1121 0.597 RCIM_Cubic>,<at 0.676 0 RCIM_Cubic>]}
        BlueCurve {keys [<at 0 1 RCIM_Cubic>,<at 1 1 RCIM_Cubic>]}
        GreenCurve {keys [<at 0 1 RCIM_Cubic>,<at 1 1 RCIM_Cubic>]}
        RedCurve {keys [<at 0 1 RCIM_Cubic>,<at 1 1 RCIM_Cubic>]}
       }})
     Scale Factor=&MakeVector4_FromLinearColor(LinearColor=$Particles.Color)
    )) {isSetParameters=true}
   /SetVariables(Particles.MainCol:LinearColor=&MultiplyLinearColor(Linear Color=<rgba 1 1 1 0.449>, Scale Factor=&MakeVector4_FromLinearColor(LinearColor=$Particles.Color))) {isSetParameters=true}
  ]
  rendererClasses [&NiagaraRibbonRendererProperties]
  renderers [
   {rendererClass=&NiagaraRibbonRendererProperties}
   {rendererClass=&NiagaraRibbonRendererProperties, rendererIndex=1}
   {rendererClass=&NiagaraRibbonRendererProperties, rendererIndex=2}
  ]
 }
 NE_Swing_01001 {
  emitterSpawnScript [&FN_RibbonIDOffsetUtility()]
  emitterUpdateScript [
   &EmitterState()
   &FN_SpawnAnimTrail(AnimTrailWidthMode=::FromSecond)
  ]
  particleSpawnScript [
   &FN_InitializeAnimTrailBoneIndecies()
   &FN_InitializeAnimTrail(
    AnimTrailWidth=1.15
    Color=&MakeLinearColorFromVectorAndFloat(
     "Vector (RGB)"=&MakeVectorFromLinearColorRGB(Color=<rgba 1 1 1 1>)
     "Float (Alpha)"=&Multiply_Float(A=$System.FadeSpawn, B=1)
    )
    Lifetime=0.13
   )
  ]
  particleUpdateScript [
   &FN_UpdateAnimTrail() {module=FN_UpdateAnimTrail001}
   &DynamicMaterialParameters(
    Index 0 Param 1=1
    Index 1 Param 3=&FloatFromCurve(
     FloatCurve=DataInterface() {propertyValues {Curve {keys [<at 0 0>, {arriveTangent=1.9814, tangentMode=RCTM_User, time=1, value=1}]}}}
     CurveIndex=&ModuloFloat(A=$Emitter.Age, B=1)
     Scale Curve=3
    )
    Index 1 Param 4=1
    Index 2 Param 1=10
    Index 2 Param 2=1
    Module.Index0Float0Write=true
    Module.Index0Float1Write=true
    Module.Index0Float2Write=true
    Module.Index0Float3Write=true
    Module.Index1Float0Write=true
    Module.Index1Float1Write=true
    Module.Index1Float2Write=true
    Module.Index1Float3Write=true
    Module.Index2Float0Write=true
    Module.Index2Float1Write=true
    Module.Index2Float2Write=true
    Module.Index2Float3Write=true
    Write Parameter Index 1=::NewEnumerator1
    Write Parameter Index 2=::NewEnumerator1
   )
   &ScaleColor(Scale Alpha=$System.FadeSystem, ScaleA=true, ScaleRGB=true) {module=ScaleColor002}
   /SetVariables(
    Particles.FillCol:LinearColor=&MultiplyLinearColor(
     Linear Color=<rgba 0.5652 0.7305 1 0.3>
     Scale Factor=&MakeVector4_FromLinearColor(LinearColor=$Particles.Color)
    )
    Particles.FillWidth=27.1842
   ) {isSetParameters=true}
  ]
  rendererClasses [&NiagaraRibbonRendererProperties]
  renderers [{rendererClass=&NiagaraRibbonRendererProperties}]
 }
}
summary {
 systemName=NS_Pickaxe_TouchToken_Trail
 emitters [
  {emitterName=NE_Swing_01, rendererClasses [&NiagaraRibbonRendererProperties]}
  {emitterName=NE_Swing_01001, rendererClasses [&NiagaraRibbonRendererProperties]}
 ]
 userVariables [
  User.bIsFrontEnd {type=NiagaraBool}
  User.bIsFrontEndPreview {type=NiagaraBool}
  User.PhysicalSurface {type=NiagaraInt32}
  User.RibbonIDOffset {type=NiagaraInt32}
  User.SwingSectionIndex {type=NiagaraInt32}
 ]
}
